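; Burst gameplay cue, named for a sniper mark-execute elimination, that plays role-specific sounds through the instigating player pawn's sound library component.
; OnBurst runs the parent burst behavior, then caches the cue's instigator as a Fortnite player pawn.
; If that pawn is friendly to the local player or locally viewed, it plays the instigator sound.
; If the cue's source object is a local player controller, it plays the target sound on that same pawn's component.
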(fn OnBurst(EitherVFX-preferthisversion)
  (params (MyTarget :type Actor%20Object%20Reference) (Parameters :type Gameplay%20Cue%20Parameters%20Structure%20%28by%20ref%29) (ParticleComponents :type Array%20of%20FXSystem%20Component%20Object%20References) (AudioComponents :type Array%20of%20Audio%20Component%20Object%20References) (BurstCameraShakeInstance :type Camera%20Shake%20Base%20Object%20Reference) (BurstDecalInstance :type Decal%20Actor%20Object%20Reference))
  (bind (_normalizedmagnitude _rawmagnitude _effectcontext _matchedtagname _originaltag _aggregatedsourcetags _aggregatedtargettags _location _normal _instigator _effectcauser _sourceobject) (Ability|GameplayCue|BreakGameplayCueParameters Parameters))
  (bind _instigatingpawn (LocalVariables|GetInstigatingPawn))
  (bind _soundlibrarycomponent (Variables|/Script/FortniteGame.FortPawn|GetSoundLibraryComponent _instigatingpawn))
  (|CallParentFunction MyTarget Parameters ParticleComponents AudioComponents BurstCameraShakeInstance BurstDecalInstance)
  (bind _asplayer_pawn_athena (Utilities|Casting|CastToPlayerPawn_Athena _instigator))
  (bind _output_get (LocalVariables|SetInstigatingPawn _asplayer_pawn_athena))
  (Utilities|FlowControl|Sequence
    (:then_0
      (if (or (Utilities|Enum|Equal(Enum) (Team|GetTeamAffiliationforActortoLocalPlayer _output_get true true) "Friendly") (Pawn|IsLocallyViewed _output_get))
        (Utilities|IsValid (Variables|/Script/FortniteGame.FortPawn|GetSoundLibraryComponent _output_get)
          (:Is%20Valid
            ; Instigator
            (SoundLibrary|PlayEvent (Variables|/Script/FortniteGame.FortPawn|GetSoundLibraryComponent _output_get) (Variables|Default|GetInstigatorSound))
            (return)))))
    (:then_1
      (bind _asplayer_controller (Utilities|Casting|CastToPlayerController _sourceobject))
      (if (Pawn|IsLocalController _asplayer_controller)
        (Utilities|IsValid _soundlibrarycomponent
          (:Is%20Valid
            ; Target Sound
            (SoundLibrary|PlayEvent _soundlibrarycomponent (Variables|Default|GetTargetSound))))))))
