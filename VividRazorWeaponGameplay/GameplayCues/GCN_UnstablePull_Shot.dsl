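; Burst gameplay cue for the Unstable Pull shot that attaches and orients its Niagara effect when it fires.
; After the parent burst logic, it feeds the source weapon's mesh into the first particle component as "TargetWeapon".
; GetFacingVector approximates shot direction from the pawn's eye rotation (flagged as not always correct), and the effect rotates to match.
; A physics force branch, gated on high effects quality, is a todo that only computes the facing vector.
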(fn GetFacingVector
  (params (Pawn :type Fort%20Pawn%20Object%20Reference) (Params :type Gameplay%20Cue%20Parameters%20Structure))
  (bind (_outlocation _outrotation) (Actor|GetActorEyesViewPoint Pawn))
  ; todo - this isn't right always - ideally want the direction the gun was shot in
  (return (Math|Rotator|GetRotationXVector _outrotation)))

(fn OnBurst(EitherVFX-preferthisversion)
  (params (MyTarget :type Actor%20Object%20Reference) (Parameters :type Gameplay%20Cue%20Parameters%20Structure%20%28by%20ref%29) (ParticleComponents :type Array%20of%20FXSystem%20Component%20Object%20References) (AudioComponents :type Array%20of%20Audio%20Component%20Object%20References) (BurstCameraShakeInstance :type Camera%20Shake%20Base%20Object%20Reference) (BurstDecalInstance :type Decal%20Actor%20Object%20Reference))
  (bind (_normalizedmagnitude _rawmagnitude _effectcontext _matchedtagname _originaltag _aggregatedsourcetags _aggregatedtargettags _location _normal _instigator _effectcauser _sourceobject) (Ability|GameplayCue|BreakGameplayCueParameters Parameters))
  (|CallParentFunction MyTarget Parameters ParticleComponents AudioComponents BurstCameraShakeInstance BurstDecalInstance)
  (Utilities|FlowControl|Sequence
    (:then_0
      (bind _asfort_weapon (Utilities|Casting|CastToFortWeapon _sourceobject))
      (bind _asniagara_particle_system_component (Utilities|Casting|CastToNiagaraComponent (Utilities|Array|Get(acopy) ParticleComponents)))
      (Niagara|SetNiagaraSkeletalMeshComponent _asniagara_particle_system_component "TargetWeapon" (Visual|GetWeaponMesh _asfort_weapon))
      (bind _fortpawn (Class|FortQuestBlueprintLibrary|GetFortPawnFromObject MyTarget)
        (:Valid
          (bind _outfwdvector (CallFunction|GetFacingVector _fortpawn (Variables|GetParameters)))
          (Transformation|SetWorldRotation _asniagara_particle_system_component (Math|Conversions|RotationFromXVector _outfwdvector) false false))))
    (:then_1
      ;[0.7031 0.2411 0 0.5] Physics force - todo
      (bind _returnvalue (Settings|GetGameUserSettings))
      (if (>= (Settings|GetVisualEffectQuality _returnvalue) 3)
        (bind _fortpawn_2 (Class|FortQuestBlueprintLibrary|GetFortPawnFromObject (Variables|GetMyTarget))
          (:Valid
            (CallFunction|GetFacingVector _fortpawn_2 (Variables|GetParameters))))))
    (:then_2
      (return))))
